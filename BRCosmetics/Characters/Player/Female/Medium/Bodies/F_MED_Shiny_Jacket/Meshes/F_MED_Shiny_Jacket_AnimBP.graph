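# Animation Blueprint for the female medium Shiny Jacket body cosmetic, following the attached parent mesh and adding corrective and physics motion.
# The copied parent pose passes through additive bone-space offsets on both dyn_glute bones, weighted by left and right thigh-rotation alphas.
# A RigidBody node then simulates in spine_02 space, borrowing the Rebirth Soldier jacket physics asset, with overridden gravity and scaled, clamped component motion.
# The event graph holds no logic.
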AnimBlueprint /BRCosmetics/Characters/Player/Female/Medium/Bodies/F_MED_Shiny_Jacket/Meshes/F_MED_Shiny_Jacket_AnimBP
target_skeleton=&/F_MED_Shiny_Jacket_Skeleton
graphs [
 AnimGraph {nodes [
   n1 = "Transform (Modify) Bone - Bone: dyn_glute_l"(
    ComponentPose=Local To Component(LocalPose=Copy Pose From Mesh() {bUseAttachedParent=true})
    Alpha=Get ThighLeftRotationAlpha() -> (ThighLeftRotationAlpha:float)
    Translation="-3.750,0.875,-0.9"
    Rotation="0.000,0.000,-55"
    Scale="1.25,1.25,1.25"
   ) {
    rotationMode=BMM_Additive
    rotationSpace=BCS_BoneSpace
    scaleMode=BMM_Additive
    scaleSpace=BCS_BoneSpace
    translationMode=BMM_Additive
    translationSpace=BCS_BoneSpace
    alphaBoolBlend {blendOption=Linear}
    boneToModify {boneName=dyn_glute_l}
    rotation {roll=-55}
    scale=<1.25 1.25 1.25>
    translation=<-3.75 0.875 -0.9>
   }
   n2 = "Transform (Modify) Bone - Bone: dyn_glute_r"(
    ComponentPose=@n1
    Alpha=Get ThighRightRotationAlpha() -> (ThighRightRotationAlpha:float)
    Translation="-3.75,0,0"
    Rotation="0.000,0.000,-55"
    Scale="1.3,1.3,1.3"
   ) {
    rotationMode=BMM_Additive
    rotationSpace=BCS_BoneSpace
    scaleMode=BMM_Additive
    scaleSpace=BCS_BoneSpace
    translationMode=BMM_Additive
    translationSpace=BCS_BoneSpace
    alphaBoolBlend {blendOption=Linear}
    boneToModify {boneName=dyn_glute_r}
    rotation {roll=-55}
    scale=<1.3 1.3 1.3>
    translation {x=-3.75}
   }
   n3 = RigidBody(ComponentPose=@n2) {
    bClampLinearTranslationLimitToRefPose=true
    bOverrideWorldGravity=true
    overridePhysicsAsset=/Game/Characters/Player/Female/Medium/Bodies/F_MED_Rebirth_Soldier/Meshes/F_MED_Rebirth_Soldier_Jacket
    simulationSpace=BaseBoneSpace
    alphaBoolBlend {blendOption=Linear}
    baseBoneRef {boneName=spine_02}
    componentAppliedLinearAccClamp=<3000 3000 2800>
    componentLinearAccScale=<0.4 0.4 0.2>
    componentLinearVelScale=<0.4 0.4 0.2>
    overrideWorldGravity {z=-980}
   }
   Output Pose(Result=Component To Local(ComponentPose=@n3)) {layerGroup=None}
  ]}
 EventGraph {nodes [Event BlueprintUpdateAnimation()]}
]
variables []
