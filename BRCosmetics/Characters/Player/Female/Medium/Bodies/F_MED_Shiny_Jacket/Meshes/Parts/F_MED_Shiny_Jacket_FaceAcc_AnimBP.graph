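# Anim Blueprint for the face accessory mesh of the female Shiny Jacket cosmetic, adding secondary motion to hair and bangs.
# It copies the attached parent's pose, then runs serial single-bone AnimDynamics nodes over left and right strands in root-relative space.
# Blend alpha rises toward strand tips; bangs use much stiffer springs, lower gravity and less inherited component motion than hair.
# The event graph has no logic.
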AnimBlueprint /BRCosmetics/Characters/Player/Female/Medium/Bodies/F_MED_Shiny_Jacket/Meshes/Parts/F_MED_Shiny_Jacket_FaceAcc_AnimBP
target_skeleton=&/F_MED_Shiny_Jacket_FaceAcc_Skeleton
graphs [
 AnimGraph {nodes [
   n1 = Anim Dynamics - Bone: hair1_0_l!(ComponentPose=Local To Component(LocalPose=Copy Pose From Mesh() {bUseAttachedParent=true}), ExternalForce="0,0,0") -> (Pose) {
    alpha=0.5
    angularDampingOverride=0.9999
    angularSpringConstant=100
    bAngularSpring=true
    bGravityOverrideInSimSpace=true
    bOverrideAngularDamping=true
    gravityScale=0.75
    simulationSpace=RootRelative
    alphaBoolBlend {blendOption=Linear}
    boundBone {boneName=hair1_0_l}
    componentAppliedLinearAccClamp=<300 300 300>
    componentLinearAccScale {x=0.7, z=0.8}
    componentLinearVelScale {x=0.7, z=0.8}
    physicsBodyDefinitions [{
      sphereCollisionRadius=0
      boundBone {boneName=hair1_0_l}
      boxExtents=<8 30 30>
      constraintSetup {angularLimitsMax=<0 20 0>, angularLimitsMin=<0 -20 -10>, angularTarget=<1 0 0>}
      localJointOffset=<6 0 0>
     }]
   }
   n2 = Anim Dynamics - Bone: hair1_1_l!(ComponentPose=@n1, ExternalForce="0,0,0") -> (Pose) {
    alpha=0.55
    angularDampingOverride=0.9999
    angularSpringConstant=100
    bAngularSpring=true
    bGravityOverrideInSimSpace=true
    bOverrideAngularDamping=true
    gravityScale=0.75
    simulationSpace=RootRelative
    alphaBoolBlend {blendOption=Linear}
    boundBone {boneName=hair1_1_l}
    componentAppliedLinearAccClamp=<300 300 300>
    componentLinearAccScale {x=0.7, z=0.8}
    componentLinearVelScale {x=0.7, z=0.8}
    physicsBodyDefinitions [{
      sphereCollisionRadius=0
      boundBone {boneName=hair1_1_l}
      boxExtents=<5 20 20>
      constraintSetup {angularLimitsMax=<0 20 0>, angularLimitsMin=<0 -20 -10>, angularTarget=<1 0 0>}
      localJointOffset=<3 0 0>
     }]
   }
   n3 = Anim Dynamics - Bone: hair1_2_l!(ComponentPose=@n2, ExternalForce="0,0,0") -> (Pose) {
    alpha=0.6
    angularDampingOverride=0.9999
    angularSpringConstant=100
    bAngularSpring=true
    bGravityOverrideInSimSpace=true
    bOverrideAngularDamping=true
    gravityScale=0.75
    simulationSpace=RootRelative
    alphaBoolBlend {blendOption=Linear}
    boundBone {boneName=hair1_2_l}
    componentAppliedLinearAccClamp=<300 300 300>
    componentLinearAccScale {x=0.7, z=0.8}
    componentLinearVelScale {x=0.7, z=0.8}
    physicsBodyDefinitions [{
      sphereCollisionRadius=0
      boundBone {boneName=hair1_2_l}
      boxExtents=<4 20 20>
      constraintSetup {angularLimitsMax=<0 20 0>, angularLimitsMin=<0 -30 -10>, angularTarget=<1 0 0>}
      localJointOffset=<2 0 0>
     }]
   }
   n4 = Anim Dynamics - Bone: hair1_3_l!(ComponentPose=@n3, ExternalForce="0,0,0") -> (Pose) {
    alpha=0.7
    angularDampingOverride=0.9999
    angularSpringConstant=100
    bAngularSpring=true
    bGravityOverrideInSimSpace=true
    bOverrideAngularDamping=true
    gravityScale=0.75
    simulationSpace=RootRelative
    alphaBoolBlend {blendOption=Linear}
    boundBone {boneName=hair1_3_l}
    componentAppliedLinearAccClamp=<300 300 300>
    componentLinearAccScale {x=0.7, z=0.8}
    componentLinearVelScale {x=0.7, z=0.8}
    physicsBodyDefinitions [{
      sphereCollisionRadius=0
      boundBone {boneName=hair1_3_l}
      boxExtents=<4 10 10>
      constraintSetup {angularLimitsMax=<0 20 0>, angularLimitsMin=<0 -30 -10>, angularTarget=<1 0 0>}
      localJointOffset=<2 0 0>
     }]
   }
   n5 = Anim Dynamics - Bone: hair2_0_l!(ComponentPose=@n4, ExternalForce="0,0,0") -> (Pose) {
    alpha=0.5
    angularDampingOverride=0.9999
    angularSpringConstant=100
    bAngularSpring=true
    bGravityOverrideInSimSpace=true
    bOverrideAngularDamping=true
    gravityScale=0.75
    simulationSpace=RootRelative
    alphaBoolBlend {blendOption=Linear}
    boundBone {boneName=hair2_0_l}
    componentAppliedLinearAccClamp=<300 300 300>
    componentLinearAccScale=<0.7 0.5 0.65>
    componentLinearVelScale=<0.7 0.5 0.65>
    physicsBodyDefinitions [{
      sphereCollisionRadius=0
      boundBone {boneName=hair2_0_l}
      boxExtents=<8 30 30>
      constraintSetup {angularLimitsMax=<0 20 0>, angularLimitsMin=<0 -20 -10>, angularTarget=<1 0 0>}
      localJointOffset=<6 0 0>
     }]
   }
   n6 = Anim Dynamics - Bone: hair2_1_l!(ComponentPose=@n5, ExternalForce="0,0,0") -> (Pose) {
    alpha=0.55
    angularDampingOverride=0.9999
    angularSpringConstant=100
    bAngularSpring=true
    bGravityOverrideInSimSpace=true
    bOverrideAngularDamping=true
    gravityScale=0.75
    simulationSpace=RootRelative
    alphaBoolBlend {blendOption=Linear}
    boundBone {boneName=hair2_1_l}
    componentAppliedLinearAccClamp=<300 300 300>
    componentLinearAccScale=<0.55 0.5 0.65>
    componentLinearVelScale=<0.55 0.5 0.65>
    physicsBodyDefinitions [{
      sphereCollisionRadius=0
      boundBone {boneName=hair2_1_l}
      boxExtents=<5 20 20>
      constraintSetup {angularLimitsMax=<0 20 0>, angularLimitsMin=<0 -20 -10>, angularTarget=<1 0 0>}
      localJointOffset=<3 0 0>
     }]
   }
   n7 = Anim Dynamics - Bone: hair2_2_l!(ComponentPose=@n6, ExternalForce="0,0,0") -> (Pose) {
    alpha=0.6
    angularDampingOverride=0.9999
    angularSpringConstant=100
    bAngularSpring=true
    bGravityOverrideInSimSpace=true
    bOverrideAngularDamping=true
    gravityScale=0.75
    simulationSpace=RootRelative
    alphaBoolBlend {blendOption=Linear}
    boundBone {boneName=hair2_2_l}
    componentAppliedLinearAccClamp=<300 300 300>
    componentLinearAccScale=<0.7 0.5 0.65>
    componentLinearVelScale=<0.7 0.5 0.65>
    physicsBodyDefinitions [{
      sphereCollisionRadius=0
      boundBone {boneName=hair2_2_l}
      boxExtents=<4 20 20>
      constraintSetup {angularLimitsMax=<0 20 0>, angularLimitsMin=<0 -20 -10>, angularTarget=<1 0 0>}
      localJointOffset=<2 0 0>
     }]
   }
   n8 = Anim Dynamics - Bone: hair3_0_l!(ComponentPose=@n7, ExternalForce="0,0,0") -> (Pose) {
    alpha=0.5
    angularDampingOverride=0.9999
    angularSpringConstant=100
    bAngularSpring=true
    bGravityOverrideInSimSpace=true
    bOverrideAngularDamping=true
    gravityScale=0.75
    simulationSpace=RootRelative
    alphaBoolBlend {blendOption=Linear}
    boundBone {boneName=hair3_0_l}
    componentAppliedLinearAccClamp=<300 300 300>
    componentLinearAccScale=<0.6 0.5 0.5>
    componentLinearVelScale=<0.6 0.5 0.5>
    physicsBodyDefinitions [{
      sphereCollisionRadius=0
      boundBone {boneName=hair3_0_l}
      boxExtents=<8 30 30>
      constraintSetup {angularLimitsMax=<0 20 0>, angularLimitsMin=<0 -20 -10>, angularTarget=<1 0 0>}
      localJointOffset=<6 0 0>
     }]
   }
   n9 = Anim Dynamics - Bone: hair3_1_l!(ComponentPose=@n8, ExternalForce="0,0,0") -> (Pose) {
    alpha=0.55
    angularDampingOverride=0.9999
    angularSpringConstant=100
    bAngularSpring=true
    bGravityOverrideInSimSpace=true
    bOverrideAngularDamping=true
    gravityScale=0.75
    simulationSpace=RootRelative
    alphaBoolBlend {blendOption=Linear}
    boundBone {boneName=hair3_1_l}
    componentAppliedLinearAccClamp=<300 300 300>
    componentLinearAccScale=<0.4 0.5 0.5>
    componentLinearVelScale=<0.4 0.5 0.5>
    physicsBodyDefinitions [{
      sphereCollisionRadius=0
      boundBone {boneName=hair3_1_l}
      boxExtents=<5 20 20>
      constraintSetup {angularLimitsMax=<0 20 0>, angularLimitsMin=<0 -20 -10>, angularTarget=<1 0 0>}
      localJointOffset=<3 0 0>
     }]
   }
   n10 = Anim Dynamics - Bone: hair3_2_l!(ComponentPose=@n9, ExternalForce="0,0,0") -> (Pose) {
    alpha=0.6
    angularDampingOverride=0.9999
    angularSpringConstant=100
    bAngularSpring=true
    bGravityOverrideInSimSpace=true
    bOverrideAngularDamping=true
    gravityScale=0.75
    simulationSpace=RootRelative
    alphaBoolBlend {blendOption=Linear}
    boundBone {boneName=hair3_2_l}
    componentAppliedLinearAccClamp=<300 300 300>
    componentLinearAccScale=<0.6 0.5 0.5>
    componentLinearVelScale=<0.6 0.5 0.5>
    physicsBodyDefinitions [{
      sphereCollisionRadius=0
      boundBone {boneName=hair3_2_l}
      boxExtents=<4 20 20>
      constraintSetup {angularLimitsMax=<0 20 0>, angularLimitsMin=<0 -20 -10>, angularTarget=<1 0 0>}
      localJointOffset=<2 0 0>
     }]
   }
   n11 = Anim Dynamics - Bone: hair4_0_l!(ComponentPose=@n10, ExternalForce="0,0,0") -> (Pose) {
    alpha=0.3
    angularDampingOverride=0.9999
    angularSpringConstant=100
    bAngularSpring=true
    bGravityOverrideInSimSpace=true
    bOverrideAngularDamping=true
    gravityScale=0.75
    simulationSpace=RootRelative
    alphaBoolBlend {blendOption=Linear}
    boundBone {boneName=hair4_0_l}
    componentAppliedLinearAccClamp=<300 300 300>
    componentLinearAccScale=<0.5 0.5 0.4>
    componentLinearVelScale=<0.5 0.5 0.4>
    physicsBodyDefinitions [{
      sphereCollisionRadius=0
      boundBone {boneName=hair4_0_l}
      boxExtents=<8 30 30>
      constraintSetup {angularLimitsMax=<0 20 0>, angularLimitsMin=<0 -20 -10>, angularTarget=<1 0 0>}
      localJointOffset=<6 0 0>
     }]
   }
   n12 = Anim Dynamics - Bone: hair4_1_l!(ComponentPose=@n11, ExternalForce="0,0,0") -> (Pose) {
    alpha=0.5
    angularDampingOverride=0.9999
    angularSpringConstant=100
    bAngularSpring=true
    bGravityOverrideInSimSpace=true
    bOverrideAngularDamping=true
    gravityScale=0.75
    simulationSpace=RootRelative
    alphaBoolBlend {blendOption=Linear}
    boundBone {boneName=hair4_1_l}
    componentAppliedLinearAccClamp=<300 300 300>
    componentLinearAccScale=<0.5 0.5 0.4>
    componentLinearVelScale=<0.5 0.5 0.4>
    physicsBodyDefinitions [{
      sphereCollisionRadius=0
      boundBone {boneName=hair4_1_l}
      boxExtents=<5 20 20>
      constraintSetup {angularLimitsMax=<0 20 0>, angularLimitsMin=<0 -20 -10>, angularTarget=<1 0 0>}
      localJointOffset=<3 0 0>
     }]
   }
   n13 = Anim Dynamics - Bone: hair1_0_r!(ComponentPose=@n12, ExternalForce="0,0,0") -> (Pose) {
    alpha=0.5
    angularDampingOverride=0.9999
    angularSpringConstant=100
    bAngularSpring=true
    bGravityOverrideInSimSpace=true
    bOverrideAngularDamping=true
    gravityScale=0.75
    simulationSpace=RootRelative
    alphaBoolBlend {blendOption=Linear}
    boundBone {boneName=hair1_0_r}
    componentAppliedLinearAccClamp=<300 300 300>
    componentLinearAccScale {x=0.7, z=0.8}
    componentLinearVelScale {x=0.7, z=0.8}
    physicsBodyDefinitions [{
      sphereCollisionRadius=0
      boundBone {boneName=hair1_0_r}
      boxExtents=<8 30 30>
      constraintSetup {angularLimitsMax=<0 20 0>, angularLimitsMin=<0 -20 -10>, angularTarget=<1 0 0>}
      localJointOffset=<6 0 0>
     }]
   }
   n14 = Anim Dynamics - Bone: hair1_1_r!(ComponentPose=@n13, ExternalForce="0,0,0") -> (Pose) {
    alpha=0.55
    angularDampingOverride=0.9999
    angularSpringConstant=100
    bAngularSpring=true
    bGravityOverrideInSimSpace=true
    bOverrideAngularDamping=true
    gravityScale=0.75
    simulationSpace=RootRelative
    alphaBoolBlend {blendOption=Linear}
    boundBone {boneName=hair1_1_r}
    componentAppliedLinearAccClamp=<300 300 300>
    componentLinearAccScale {x=0.7, z=0.8}
    componentLinearVelScale {x=0.7, z=0.8}
    physicsBodyDefinitions [{
      sphereCollisionRadius=0
      boundBone {boneName=hair1_1_r}
      boxExtents=<5 20 20>
      constraintSetup {angularLimitsMax=<0 20 0>, angularLimitsMin=<0 -20 -10>, angularTarget=<1 0 0>}
      localJointOffset=<3 -0.1 0>
     }]
   }
   n15 = Anim Dynamics - Bone: hair1_2_r!(ComponentPose=@n14, ExternalForce="0,0,0") -> (Pose) {
    alpha=0.6
    angularDampingOverride=0.9999
    angularSpringConstant=100
    bAngularSpring=true
    bGravityOverrideInSimSpace=true
    bOverrideAngularDamping=true
    gravityScale=0.75
    simulationSpace=RootRelative
    alphaBoolBlend {blendOption=Linear}
    boundBone {boneName=hair1_2_r}
    componentAppliedLinearAccClamp=<300 300 300>
    componentLinearAccScale {x=0.7, z=0.8}
    componentLinearVelScale {x=0.7, z=0.8}
    physicsBodyDefinitions [{
      sphereCollisionRadius=0
      boundBone {boneName=hair1_2_r}
      boxExtents=<4 20 20>
      constraintSetup {angularLimitsMax=<0 20 0>, angularLimitsMin=<0 -30 -10>, angularTarget=<1 0 0>}
      localJointOffset=<2 0 0>
     }]
   }
   n16 = Anim Dynamics - Bone: hair1_3_r!(ComponentPose=@n15, ExternalForce="0,0,0") -> (Pose) {
    alpha=0.7
    angularDampingOverride=0.9999
    angularSpringConstant=100
    bAngularSpring=true
    bGravityOverrideInSimSpace=true
    bOverrideAngularDamping=true
    gravityScale=0.75
    simulationSpace=RootRelative
    alphaBoolBlend {blendOption=Linear}
    boundBone {boneName=hair1_3_r}
    componentAppliedLinearAccClamp=<300 300 300>
    componentLinearAccScale {x=0.7, z=0.9}
    componentLinearVelScale {x=0.7, z=0.9}
    physicsBodyDefinitions [{
      sphereCollisionRadius=0
      boundBone {boneName=hair1_3_r}
      boxExtents=<4 10 10>
      constraintSetup {angularLimitsMax=<0 20 0>, angularLimitsMin=<0 -30 -10>, angularTarget=<1 0 0>}
      localJointOffset=<2 0 0>
     }]
   }
   n17 = Anim Dynamics - Bone: hair2_0_r!(ComponentPose=@n16, ExternalForce="0,0,0") -> (Pose) {
    alpha=0.5
    angularDampingOverride=0.9999
    angularSpringConstant=100
    bAngularSpring=true
    bGravityOverrideInSimSpace=true
    bOverrideAngularDamping=true
    gravityScale=0.75
    simulationSpace=RootRelative
    alphaBoolBlend {blendOption=Linear}
    boundBone {boneName=hair2_0_r}
    componentAppliedLinearAccClamp=<300 300 300>
    componentLinearAccScale=<0.7 0.5 0.65>
    componentLinearVelScale=<0.7 0.5 0.65>
    physicsBodyDefinitions [{
      sphereCollisionRadius=0
      boundBone {boneName=hair2_0_r}
      boxExtents=<8 30 30>
      constraintSetup {angularLimitsMax=<0 20 0>, angularLimitsMin=<0 -20 -10>, angularTarget=<1 0 0>}
      localJointOffset=<6 0 0>
     }]
   }
   n18 = Anim Dynamics - Bone: hair2_1_r!(ComponentPose=@n17, ExternalForce="0,0,0") -> (Pose) {
    alpha=0.55
    angularDampingOverride=0.9999
    angularSpringConstant=100
    bAngularSpring=true
    bGravityOverrideInSimSpace=true
    bOverrideAngularDamping=true
    gravityScale=0.75
    simulationSpace=RootRelative
    alphaBoolBlend {blendOption=Linear}
    boundBone {boneName=hair2_1_r}
    componentAppliedLinearAccClamp=<300 300 300>
    componentLinearAccScale=<0.55 0.5 0.65>
    componentLinearVelScale=<0.55 0.5 0.65>
    physicsBodyDefinitions [{
      sphereCollisionRadius=0
      boundBone {boneName=hair2_1_r}
      boxExtents=<5 20 20>
      constraintSetup {angularLimitsMax=<0 20 0>, angularLimitsMin=<0 -20 -10>, angularTarget=<1 0 0>}
      localJointOffset=<3 -0.1 0>
     }]
   }
   n19 = Anim Dynamics - Bone: hair2_2_r!(ComponentPose=@n18, ExternalForce="0,0,0") -> (Pose) {
    alpha=0.6
    angularDampingOverride=0.9999
    angularSpringConstant=100
    bAngularSpring=true
    bGravityOverrideInSimSpace=true
    bOverrideAngularDamping=true
    gravityScale=0.75
    simulationSpace=RootRelative
    alphaBoolBlend {blendOption=Linear}
    boundBone {boneName=hair2_2_r}
    componentAppliedLinearAccClamp=<300 300 300>
    componentLinearAccScale=<0.7 0.5 0.65>
    componentLinearVelScale=<0.7 0.5 0.65>
    physicsBodyDefinitions [{
      sphereCollisionRadius=0
      boundBone {boneName=hair2_2_r}
      boxExtents=<4 20 20>
      constraintSetup {angularLimitsMax=<0 20 0>, angularLimitsMin=<0 -20 -10>, angularTarget=<1 0 0>}
      localJointOffset=<2 0 0>
     }]
   }
   n20 = Anim Dynamics - Bone: hair3_0_r!(ComponentPose=@n19, ExternalForce="0,0,0") -> (Pose) {
    alpha=0.5
    angularDampingOverride=0.9999
    angularSpringConstant=100
    bAngularSpring=true
    bGravityOverrideInSimSpace=true
    bOverrideAngularDamping=true
    gravityScale=0.75
    simulationSpace=RootRelative
    alphaBoolBlend {blendOption=Linear}
    boundBone {boneName=hair3_0_r}
    componentAppliedLinearAccClamp=<300 300 300>
    componentLinearAccScale=<0.6 0.5 0.5>
    componentLinearVelScale=<0.6 0.5 0.5>
    physicsBodyDefinitions [{
      sphereCollisionRadius=0
      boundBone {boneName=hair3_0_r}
      boxExtents=<8 30 30>
      constraintSetup {angularLimitsMax=<0 20 0>, angularLimitsMin=<0 -20 -10>, angularTarget=<1 0 0>}
      localJointOffset=<6 0 0>
     }]
   }
   n21 = Anim Dynamics - Bone: hair3_1_r!(ComponentPose=@n20, ExternalForce="0,0,0") -> (Pose) {
    alpha=0.55
    angularDampingOverride=0.9999
    angularSpringConstant=100
    bAngularSpring=true
    bGravityOverrideInSimSpace=true
    bOverrideAngularDamping=true
    gravityScale=0.75
    simulationSpace=RootRelative
    alphaBoolBlend {blendOption=Linear}
    boundBone {boneName=hair3_1_r}
    componentAppliedLinearAccClamp=<300 300 300>
    componentLinearAccScale=<0.4 0.5 0.5>
    componentLinearVelScale=<0.4 0.5 0.5>
    physicsBodyDefinitions [{
      sphereCollisionRadius=0
      boundBone {boneName=hair3_1_r}
      boxExtents=<5 20 20>
      constraintSetup {angularLimitsMax=<0 20 0>, angularLimitsMin=<0 -20 -10>, angularTarget=<1 0 0>}
      localJointOffset=<3 -0.1 0>
     }]
   }
   n22 = Anim Dynamics - Bone: hair3_2_r!(ComponentPose=@n21, ExternalForce="0,0,0") -> (Pose) {
    alpha=0.6
    angularDampingOverride=0.9999
    angularSpringConstant=100
    bAngularSpring=true
    bGravityOverrideInSimSpace=true
    bOverrideAngularDamping=true
    gravityScale=0.75
    simulationSpace=RootRelative
    alphaBoolBlend {blendOption=Linear}
    boundBone {boneName=hair3_2_r}
    componentAppliedLinearAccClamp=<300 300 300>
    componentLinearAccScale=<0.6 0.5 0.5>
    componentLinearVelScale=<0.6 0.5 0.5>
    physicsBodyDefinitions [{
      sphereCollisionRadius=0
      boundBone {boneName=hair3_2_r}
      boxExtents=<4 20 20>
      constraintSetup {angularLimitsMax=<0 20 0>, angularLimitsMin=<0 -20 -10>, angularTarget=<1 0 0>}
      localJointOffset=<2 0 0>
     }]
   }
   n23 = Anim Dynamics - Bone: hair4_0_r!(ComponentPose=@n22, ExternalForce="0,0,0") -> (Pose) {
    alpha=0.3
    angularDampingOverride=0.9999
    angularSpringConstant=100
    bAngularSpring=true
    bGravityOverrideInSimSpace=true
    bOverrideAngularDamping=true
    gravityScale=0.75
    simulationSpace=RootRelative
    alphaBoolBlend {blendOption=Linear}
    boundBone {boneName=hair4_0_r}
    componentAppliedLinearAccClamp=<300 300 300>
    componentLinearAccScale=<0.5 0.5 0.4>
    componentLinearVelScale=<0.5 0.5 0.4>
    physicsBodyDefinitions [{
      sphereCollisionRadius=0
      boundBone {boneName=hair4_0_r}
      boxExtents=<8 30 30>
      constraintSetup {angularLimitsMax=<0 20 0>, angularLimitsMin=<0 -20 -10>, angularTarget=<1 0 0>}
      localJointOffset=<6 0 0>
     }]
   }
   n24 = Anim Dynamics - Bone: hair4_1_r!(ComponentPose=@n23, ExternalForce="0,0,0") -> (Pose) {
    alpha=0.5
    angularDampingOverride=0.9999
    angularSpringConstant=100
    bAngularSpring=true
    bGravityOverrideInSimSpace=true
    bOverrideAngularDamping=true
    gravityScale=0.75
    simulationSpace=RootRelative
    alphaBoolBlend {blendOption=Linear}
    boundBone {boneName=hair4_1_r}
    componentAppliedLinearAccClamp=<300 300 300>
    componentLinearAccScale=<0.5 0.5 0.4>
    componentLinearVelScale=<0.5 0.5 0.4>
    physicsBodyDefinitions [{
      sphereCollisionRadius=0
      boundBone {boneName=hair4_1_r}
      boxExtents=<5 20 20>
      constraintSetup {angularLimitsMax=<0 20 0>, angularLimitsMin=<0 -20 -10>, angularTarget=<1 0 0>}
      localJointOffset=<3 -0.1 0>
     }]
   }
   n25 = Anim Dynamics - Bone: bangs1_0_r!(ComponentPose=@n24, ExternalForce="0,0,0") -> (Pose) {
    alpha=0.1
    angularDampingOverride=1
    angularSpringConstant=1500
    bAngularSpring=true
    bGravityOverrideInSimSpace=true
    bOverrideAngularDamping=true
    gravityScale=0.35
    simulationSpace=RootRelative
    alphaBoolBlend {blendOption=Linear}
    boundBone {boneName=bangs1_0_r}
    componentLinearAccScale=<0.2 0.2 0.2>
    componentLinearVelScale=<0.2 0.2 0.2>
    physicsBodyDefinitions [{
      sphereCollisionRadius=0
      boundBone {boneName=bangs1_0_r}
      boxExtents=<10 20 20>
      constraintSetup {angularLimitsMax=<0 0 0>, angularLimitsMin=<0 0 -40>, angularTarget=<0 0 0>}
      localJointOffset=<10 0 0>
     }]
   }
   n26 = Anim Dynamics - Bone: bangs1_1_r!(ComponentPose=@n25, ExternalForce="0,0,0") -> (Pose) {
    alpha=0.3
    angularDampingOverride=1
    angularSpringConstant=1500
    bAngularSpring=true
    bGravityOverrideInSimSpace=true
    bOverrideAngularDamping=true
    gravityScale=0.35
    simulationSpace=RootRelative
    alphaBoolBlend {blendOption=Linear}
    boundBone {boneName=bangs1_1_r}
    componentLinearAccScale=<0.2 0.2 0.2>
    componentLinearVelScale=<0.2 0.2 0.2>
    physicsBodyDefinitions [{
      sphereCollisionRadius=0
      boundBone {boneName=bangs1_1_r}
      boxExtents=<10 20 20>
      constraintSetup {angularLimitsMax=<0 0 0>, angularLimitsMin=<0 0 -40>, angularTarget=<0 0 0>}
      localJointOffset=<10 0 0>
     }]
   }
   n27 = Anim Dynamics - Bone: bangs1_2_r!(ComponentPose=@n26, ExternalForce="0,0,0") -> (Pose) {
    alpha=0.4
    angularDampingOverride=1
    angularSpringConstant=1500
    bAngularSpring=true
    bGravityOverrideInSimSpace=true
    bOverrideAngularDamping=true
    gravityScale=0.35
    simulationSpace=RootRelative
    alphaBoolBlend {blendOption=Linear}
    boundBone {boneName=bangs1_2_r}
    componentLinearAccScale=<0.2 0.2 0.2>
    componentLinearVelScale=<0.2 0.2 0.2>
    physicsBodyDefinitions [{
      sphereCollisionRadius=0
      boundBone {boneName=bangs1_2_r}
      boxExtents=<10 20 20>
      constraintSetup {angularLimitsMax=<0 0 0>, angularLimitsMin=<0 0 -40>, angularTarget=<0 0 0>}
      localJointOffset=<10 0 0>
     }]
   }
   n28 = Anim Dynamics - Bone: bangs2_0_r!(ComponentPose=@n27, ExternalForce="0,0,0") -> (Pose) {
    alpha=0.1
    angularDampingOverride=1
    angularSpringConstant=1500
    bAngularSpring=true
    bGravityOverrideInSimSpace=true
    bOverrideAngularDamping=true
    gravityScale=0.35
    simulationSpace=RootRelative
    alphaBoolBlend {blendOption=Linear}
    boundBone {boneName=bangs2_0_r}
    componentLinearAccScale=<0.2 0.2 0.2>
    componentLinearVelScale=<0.2 0.2 0.2>
    physicsBodyDefinitions [{
      sphereCollisionRadius=0
      boundBone {boneName=bangs2_0_r}
      boxExtents=<10 20 20>
      constraintSetup {angularLimitsMax=<0 0 0>, angularLimitsMin=<0 0 -40>, angularTarget=<0 0 0>}
      localJointOffset=<10 0 0>
     }]
   }
   n29 = Anim Dynamics - Bone: bangs2_1_r!(ComponentPose=@n28, ExternalForce="0,0,0") -> (Pose) {
    alpha=0.3
    angularDampingOverride=1
    angularSpringConstant=1500
    bAngularSpring=true
    bGravityOverrideInSimSpace=true
    bOverrideAngularDamping=true
    gravityScale=0.35
    simulationSpace=RootRelative
    alphaBoolBlend {blendOption=Linear}
    boundBone {boneName=bangs2_1_r}
    componentLinearAccScale=<0.2 0.2 0.2>
    componentLinearVelScale=<0.2 0.2 0.2>
    physicsBodyDefinitions [{
      sphereCollisionRadius=0
      boundBone {boneName=bangs2_1_r}
      boxExtents=<10 20 20>
      constraintSetup {angularLimitsMax=<0 0 0>, angularLimitsMin=<0 0 -40>, angularTarget=<0 0 0>}
      localJointOffset=<10 0 0>
     }]
   }
   n30 = Anim Dynamics - Bone: bangs2_2_r!(ComponentPose=@n29, ExternalForce="0,0,0") -> (Pose) {
    alpha=0.4
    angularDampingOverride=1
    angularSpringConstant=1500
    bAngularSpring=true
    bGravityOverrideInSimSpace=true
    bOverrideAngularDamping=true
    gravityScale=0.35
    simulationSpace=RootRelative
    alphaBoolBlend {blendOption=Linear}
    boundBone {boneName=bangs2_2_r}
    componentLinearAccScale=<0.2 0.2 0.2>
    componentLinearVelScale=<0.2 0.2 0.2>
    physicsBodyDefinitions [{
      sphereCollisionRadius=0
      boundBone {boneName=bangs2_2_r}
      boxExtents=<10 20 20>
      constraintSetup {angularLimitsMax=<0 0 0>, angularLimitsMin=<0 0 -40>, angularTarget=<0 0 0>}
      localJointOffset=<10 0 0>
     }]
   }
   n31 = Anim Dynamics - Bone: bangs2_0_l!(ComponentPose=@n30, ExternalForce="0,0,0") -> (Pose) {
    alpha=0.1
    angularDampingOverride=1
    angularSpringConstant=1500
    bAngularSpring=true
    bGravityOverrideInSimSpace=true
    bOverrideAngularDamping=true
    gravityScale=0.35
    simulationSpace=RootRelative
    alphaBoolBlend {blendOption=Linear}
    boundBone {boneName=bangs2_0_l}
    componentLinearAccScale=<0.2 0.2 0.2>
    componentLinearVelScale=<0.2 0.2 0.2>
    physicsBodyDefinitions [{
      sphereCollisionRadius=0
      boundBone {boneName=bangs2_0_l}
      boxExtents=<10 20 20>
      constraintSetup {angularLimitsMax=<0 0 0>, angularLimitsMin=<0 0 -40>, angularTarget=<0 0 0>}
      localJointOffset=<10 0 0>
     }]
   }
   n32 = Anim Dynamics - Bone: bangs2_1_l!(ComponentPose=@n31, ExternalForce="0,0,0") -> (Pose) {
    alpha=0.3
    angularDampingOverride=1
    angularSpringConstant=1500
    bAngularSpring=true
    bGravityOverrideInSimSpace=true
    bOverrideAngularDamping=true
    gravityScale=0.35
    simulationSpace=RootRelative
    alphaBoolBlend {blendOption=Linear}
    boundBone {boneName=bangs2_1_l}
    componentLinearAccScale=<0.2 0.2 0.2>
    componentLinearVelScale=<0.2 0.2 0.2>
    physicsBodyDefinitions [{
      sphereCollisionRadius=0
      boundBone {boneName=bangs2_1_l}
      boxExtents=<10 20 20>
      constraintSetup {angularLimitsMax=<0 0 0>, angularLimitsMin=<0 0 -40>, angularTarget=<0 0 0>}
      localJointOffset=<10 0 0>
     }]
   }
   n33 = Anim Dynamics - Bone: bangs2_2_l!(ComponentPose=@n32, ExternalForce="0,0,0") -> (Pose) {
    alpha=0.4
    angularDampingOverride=1
    angularSpringConstant=1500
    bAngularSpring=true
    bGravityOverrideInSimSpace=true
    bOverrideAngularDamping=true
    gravityScale=0.35
    simulationSpace=RootRelative
    alphaBoolBlend {blendOption=Linear}
    boundBone {boneName=bangs2_2_l}
    componentLinearAccScale=<0.2 0.2 0.2>
    componentLinearVelScale=<0.2 0.2 0.2>
    physicsBodyDefinitions [{
      sphereCollisionRadius=0
      boundBone {boneName=bangs2_2_l}
      boxExtents=<10 20 20>
      constraintSetup {angularLimitsMax=<0 0 0>, angularLimitsMin=<0 0 -40>, angularTarget=<0 0 0>}
      localJointOffset=<10 0 0>
     }]
   }
   n34 = Anim Dynamics - Bone: bangs1_0_l!(ComponentPose=@n33, ExternalForce="0,0,0") -> (Pose) {
    alpha=0.1
    angularDampingOverride=1
    angularSpringConstant=1500
    bAngularSpring=true
    bGravityOverrideInSimSpace=true
    bOverrideAngularDamping=true
    gravityScale=0.35
    simulationSpace=RootRelative
    alphaBoolBlend {blendOption=Linear}
    boundBone {boneName=bangs1_0_l}
    componentLinearAccScale=<0.2 0.2 0.2>
    componentLinearVelScale=<0.2 0.2 0.2>
    physicsBodyDefinitions [{
      sphereCollisionRadius=0
      boundBone {boneName=bangs1_0_l}
      boxExtents=<10 20 20>
      constraintSetup {angularLimitsMax=<0 0 0>, angularLimitsMin=<0 0 -40>, angularTarget=<0 0 0>}
      localJointOffset=<10 0 0>
     }]
   }
   n35 = Anim Dynamics - Bone: bangs1_1_l!(ComponentPose=@n34, ExternalForce="0,0,0") -> (Pose) {
    alpha=0.3
    angularDampingOverride=1
    angularSpringConstant=1500
    bAngularSpring=true
    bGravityOverrideInSimSpace=true
    bOverrideAngularDamping=true
    gravityScale=0.35
    simulationSpace=RootRelative
    alphaBoolBlend {blendOption=Linear}
    boundBone {boneName=bangs1_1_l}
    componentLinearAccScale=<0.2 0.2 0.2>
    componentLinearVelScale=<0.2 0.2 0.2>
    physicsBodyDefinitions [{
      sphereCollisionRadius=0
      boundBone {boneName=bangs1_1_l}
      boxExtents=<10 20 20>
      constraintSetup {angularLimitsMax=<0 0 0>, angularLimitsMin=<0 0 -40>, angularTarget=<0 0 0>}
      localJointOffset=<10 0 0>
     }]
   }
   n36 = Anim Dynamics - Bone: bangs1_2_l!(ComponentPose=@n35, ExternalForce="0,0,0") -> (Pose) {
    alpha=0.4
    angularDampingOverride=1
    angularSpringConstant=1500
    bAngularSpring=true
    bGravityOverrideInSimSpace=true
    bOverrideAngularDamping=true
    gravityScale=0.35
    simulationSpace=RootRelative
    alphaBoolBlend {blendOption=Linear}
    boundBone {boneName=bangs1_2_l}
    componentLinearAccScale=<0.2 0.2 0.2>
    componentLinearVelScale=<0.2 0.2 0.2>
    physicsBodyDefinitions [{
      sphereCollisionRadius=0
      boundBone {boneName=bangs1_2_l}
      boxExtents=<10 20 20>
      constraintSetup {angularLimitsMax=<0 0 0>, angularLimitsMin=<0 0 -40>, angularTarget=<0 0 0>}
      localJointOffset=<10 0 0>
     }]
   }
   Output Pose(Result=Component To Local(ComponentPose=@n36)) {layerGroup=None}
  ]}
 EventGraph {nodes [
   TryGetPawnOwner()
   Event BlueprintUpdateAnimation()
  ]}
]
variables []
